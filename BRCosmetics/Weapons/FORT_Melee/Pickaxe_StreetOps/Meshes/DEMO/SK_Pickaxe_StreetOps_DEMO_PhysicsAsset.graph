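# Physics asset for the StreetOps DEMO pickaxe skeletal mesh, a cosmetic two-handed melee weapon.
# A single capsule body on the Melee_TwoHanded_Handle bone serves as the collision shape, offset and rotated relative to that bone.
# It defines no constraints, so there are no jointed parts to simulate.
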PhysicsAsset /BRCosmetics/Weapons/FORT_Melee/Pickaxe_StreetOps/Meshes/DEMO/SK_Pickaxe_StreetOps_DEMO_PhysicsAsset
bodies {Melee_TwoHanded_Handle [{shapeType=Capsule, length=64.7555, radius=30.0646, center=<0.8085 -3.3656 28.3958>, rotation=<rot -2.3139 -90 -0.0059>}]}
constraints []
